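# Master additive unlit Niagara sprite material for spark particles, drawn as capsules or optional teardrops stretched by world position offset.
# A sphere mask on the shape's UVs, with optional end fade and power, drives glow and opacity with particle color.
# Emissive gets HDR remapping and opacity gets depth fade, both bypassed on mobile, plus camera-depth fade.
# Previous-frame offset subtracts particle velocity times delta time for motion vectors.
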Material /Game/Effects/Niagara/MasterMaterials/Sparks/M_Capsule_Spark_Niagara_Master
graph {
 properties {
  bEnableResponsiveAA=true
  bUsedWithNiagaraSprites=true
  blendMode=BLEND_Additive
  pixelDepthOffsetMode=PDOM_Legacy
  shadingModel=MSM_Unlit
  usedShadingModels=Unlit
 }
 nodes [
  n1 = ParticleColor()
  n2 = DynamicParameter() {defaultValue=<rgba 2 20 1 1>, paramNames [Power,DepthFade,HDRMult,BottomFade]}
  n3 = LinearInterpolate(A=@n2.BottomFade, Alpha=ComponentMask(None=TextureCoordinate()) {r=true}) {constA=1}
  n4 = DynamicParameter() {parameterIndex=1, defaultValue=<rgba 20 500 10 1>, paramNames [TeardropMinSpeed, TeardropMaxSpeed, TeardropMaxStretch, Param4]}
  n5 = MaterialFunctionCall!(
   Max Stretch Factor=@n4.TeardropMaxStretch
   Min Tear Drop Width=ScalarParameter() {defaultValue=0.1, parameterName=TeardropMinWidth}
   Minimum Speed=@n4.TeardropMaxStretch
   Max Speed=@n4.TeardropMaxStretch
   ------ Particle Defaults Below ------
   Particle Speed
   Particle Direction
   Particle Position
   Particle Size
   UV Add
  ) {materialFunction=/Engine/Functions/Engine_MaterialFunctions02/WorldPositionOffset/Sprite_TearDrop}
  n6 = StaticBool()
  n7 = MaterialFunctionCall!(
   Particle Position
   Normalized Particle Direction
   Capsule Diameter
   Capsule Pivot Alignment
   Scale Along Velocity Vector
   ---------Scale Over Velocity Parameters---------
   Particle Speed
   Minimum Speed
   Max Speed
   Minimum Scale Factor=MaterialFunctionCall(Float2=ParticleSize()) {materialFunction=&BreakOutFloat2Components}.G
   Maximum Scale Factor
   ---------- Advanced -----------
   Cap Size Multiplier=ScalarParameter() {defaultValue=0.75, parameterName=Cap Size Multiplier, sliderMax=1}
   Use Camera Offset=@n6
   Use Advanced Features=@n6
   World Position to Negate
  ) {materialFunction=/Engine/Functions/Engine_MaterialFunctions02/WorldPositionOffset/Sprite_Capsule}
  n8 = Constant() {r=0.5}
  n9 = SphereMask(
   A=StaticSwitchParameter(True=@n5.Texture Coordinates, False=@n7.Texture Coordinates) {parameterName=UseTeardropShape}
   B=@n8
   Radius=@n8
   Hardness=ScalarParameter() {parameterName=SphereMaskHardness}
  ) {hardnessPercent=0}
  n10 = Power(Base=StaticSwitchParameter(True=Multiply(A=@n3, B=@n9), False=@n9) {parameterName=UseEndFade}, Exp=@n2.Power)
  n11 = Multiply(
   A=@n1.RGB
   B=LinearInterpolate(A=ScalarParameter() {parameterName=Glow Dark Value}, B=ScalarParameter() {defaultValue=1, parameterName=Glow Bright Value}, Alpha=@n10)
  )
  n12 = Clamp(None=@n11)
  MaterialFunctionCall(Emissive=ShadingPathSwitch(
    Default=MaterialFunctionCall(LDR Input=@n12, HDR Tint=@n11, HDR Intensity Multiplier=@n2.HDRMult) {materialFunction=&DeriveHDRfromLDR}.RemappedRange
    Mobile=@n11
   )) {materialFunction=&MF_FX_EmissiveAdjustments}
  n13 = DepthFade(Opacity=@n10, FadeDistance=@n2.DepthFade)
  Multiply(
   A=Multiply(A=@n1.A, B=ShadingPathSwitch(Default=@n13, Forward=@n10, Mobile=@n10))
   B=MaterialFunctionCall(
    Fade Length=ScalarParameter() {parameterName=Camera depth Fade length}
    Fade Offset=ScalarParameter() {parameterName=Camera depth Fade Offset}
   ) {materialFunction=&CameraDepthFade}.Result
  )
  n14 = Multiply(A=Multiply(A=ParticleDirection(), B=ParticleSpeed()), B=DeltaTime())
  n15 = StaticSwitchParameter(True=@n5.World Position Offset, False=@n7.World Position Offset) {parameterName=UseTeardropShape}
  PreviousFrameSwitch(Current Frame=@n15, Previous Frame=Subtract(A=@n15, B=Multiply(A=@n14)))
 ]
}
